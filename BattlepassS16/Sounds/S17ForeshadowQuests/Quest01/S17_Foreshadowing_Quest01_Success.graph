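# Sound cue for the success sound of the first Season 17 foreshadowing quest in the Season 16 Battle Pass content.
# A Random node picks one of three success wave variations with equal weighting.
# It plays under the quests sound class and uses the shared foreshadowing-quest attenuation asset for spatialization and falloff.
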SoundCue /BattlepassS16/Sounds/S17ForeshadowQuests/Quest01/S17_Foreshadowing_Quest01_Success
root=Random_0
nodes [
 {id=Random_0, children [WavePlayer_0,WavePlayer_1,WavePlayer_2], properties {weights [1,1,1]}}
 {id=WavePlayer_0, properties {soundWaveAssetPtr=&/S17_Foreshadowing_Quest01_Success01}}
 {id=WavePlayer_1, properties {soundWaveAssetPtr=&/S17_Foreshadowing_Quest01_Success02}}
 {id=WavePlayer_2, properties {soundWaveAssetPtr=&/S17_Foreshadowing_Quest01_Success03}}
]
properties {
 attenuationSettings=/BattlepassS16/Sounds/S17ForeshadowQuests/S17_Foreshadowing_Quest_Att
 maxDistance=2097152
 soundClassObject=&Fort_Quests
 volumeMultiplier=1
}
